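# Idle Niagara effect for the Polished Jade pickaxe: teal flame-like meshes, glow and decal sprites attached to its skeletal mesh, with purple variant colors.
# Seven partitioned particles follow sockets with per-index scale and material values, feeding a smoothed position to materials as trail data.
# A single-particle emitter adds another sprite and mesh layer.
# Front-end flags switch to looping, curve-faded emitter copies; otherwise alpha follows the system timeline, and time of day scales color.
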NiagaraSystem /BRCosmetics/Weapons/FORT_Melee/Pickaxe_PolishedJade/FX/NS_Pickaxe_PolishedJade_Idle
dependencies {
 usedDataInterfaces [
  &NiagaraDataInterfaceArrayFloat
  &NiagaraDataInterfaceArrayInt32
  &NiagaraDataInterfaceArrayFloat3
  &NiagaraDataInterfaceCurve
  &NiagaraDataInterfaceSkeletalMesh
 ]
 usedDynamicInputs [
  &SetBoolByFloatComparison
  &FN_Select_Float_from_Time_Of_Day
  &FN_Select_Vector_from_Time_Of_Day
 ]
 usedModules [
  &self:SystemSpawnScript.SystemScriptSource.SystemScriptGraph.NiagaraNodeAssignment_0.SetVariables_#1
  &SystemState
  &Timeline
  &FN_VariantManager
  &self:SystemSpawnScript.SystemScriptSource.SystemScriptGraph.NiagaraNodeAssignment_1.SetVariables_#2
 ]
}
emitters {
 BlowingParticles {
  emitterUpdateScript [
   &EmitterState(
    Inactive Response=::NewEnumerator1
    Life Cycle Mode=::NewEnumerator1
    Loop Behavior=::NewEnumerator1
    Loop Duration Mode=::NewEnumerator1
    UseLoopDelay=false
   )
   &SpawnBurst_Instantaneous(
    Spawn Count=&MakeCustomIntFromBool(Bool=&Boolean_OrOperation(Bool1=$User.bIsFrontEnd, Bool2=$User.bIsFrontEndPreview), True INT=0, False INT=7)
    Use Spawn Probability=false
    UseLoopCountLimit=false
   )
  ]
  particleSpawnScript [
   &PartitionParticles(Number Of Partitions=7)
   &InitializeParticle(
    Color=&MultiplyLinearColor(
     Linear Color=&VariantColorSelector(Variant Amount=0, Original Color=<rgba 0.0018 0.0885 0.0857 1>, Variant 1 Color=<rgba 0.125 0.0182 0.1125 1>, Variant 2 Color=?)
     Scale Factor=&MakeFloat4FromVectorAndFloat("Vector (RGB)"=$System.ToD, "Float (Alpha)"=1)
    )
    Lifetime Max=15
    Lifetime Min=12
    Lifetime Mode=::NewEnumerator1
    Mesh Scale Mode=::NewEnumerator3
    Mesh Uniform Scale=&SelectFloatFromArray(
     Direct Array Index=$Particles.PartitionParticles.Partition
     Float Selection Array=DataInterface() {propertyValues {FloatData [0.6,0.35,0.2,0.7,0.5,0.9,0.4]}}
    )
    Sprite Size Mode=::NewEnumerator3
    Uniform Sprite Size=7
    UsePositionOffset=false
    Write Material Random=true
   )
   &InitialMeshOrientation(Mesh Axis To Orient=<0 0 -1>, Mesh Orientation Mode=::NewEnumerator2, Orientation Vector=<0 1 0>, Use Rotation Vector=false)
   /SetVariables(Particles.VisibilityTag=&SelectIntFromArray(Direct Array Index=$Particles.PartitionParticles.Partition, Int Selection Array=DataInterface() {propertyValues {IntData [0,0,0,1,1,0,0]}})) {isSetParameters=true, module=SetVariables_#3}
   &DynamicMaterialParameters(
    Index 0 Param 1=&RandomRangeFloat(Minimum=0, Maximum=1)
    Index 0 Param 2=&SelectFloatFromArray(
     Direct Array Index=$Particles.PartitionParticles.Partition
     Float Selection Array=DataInterface() {propertyValues {FloatData [2.2,1,0.5,1.5,0.8,2.5,1.8]}}
    )
    Module.Index0Param0Write=true
    Module.Index0Param1Write=true
    Module.Index0Param2Write=true
    Module.Index0Param3Write=true
   )
  ]
  particleUpdateScript [
   &ParticleState(Kill Particles When Lifetime Has Elapsed=false, Let Infinitely Lived Particles Die When Emitter Deactivates=true)
   /SetVariables(Particles.AlphaScale:NiagaraFloat=&Lerp_Float(A=0, B=1, Alpha=$System.Timeline.NormalizedTime)) {isSetParameters=true, module=SetVariables_#4}
   &ScaleMeshSize(Scale Factor=<1 1 2>)
   &SolveForcesAndVelocity(Clamp Velocity=false, Limit Acceleration=false)
   &SkeletalMeshLocation(
    Bone / Socket Index=&ReturnExecIndex()
    MaskSpawnGroup=false
    Mesh Sampling Type=::NewEnumerator5
    Skeletal Mesh=$System.Skeletal Mesh
    Socket Sampling Mode=::NewEnumerator5
   )
   &ScaleColor(Scale Alpha=$Particles.AlphaScale, Scale RGB=&VectorFromFloat(Value=3.5), ScaleA=true, ScaleRGB=true)
   /SetVariables(Particles.Color_02:LinearColor=&MultiplyLinearColor(
     Linear Color=&MultiplyLinearColor(
      Linear Color=&VariantColorSelector(Variant Amount=0, Original Color=<rgba 0.0018 0.0885 0.0857 1>, Variant 1 Color=<rgba 0.2812 0.041 0.2452 1>, Variant 2 Color=?)
      Scale Factor=&MakeFloat4FromVectorAndFloat("Vector (RGB)"=$System.ToD, "Float (Alpha)"=1)
     )
     Scale Factor=&MakeFloat4FromVectorAndFloat("Vector (RGB)"=<1 1 1>, "Float (Alpha)"=$Particles.AlphaScale)
    )) {isSetParameters=true, module=SetVariables_#5}
   /SetVariables(Particles.Color_03:LinearColor=&MultiplyLinearColor(
     Linear Color=&MultiplyLinearColor(
      Linear Color=&VariantColorSelector(Variant Amount=0, Original Color=<rgba 0.2203 0.9 0.7754 1>, Variant 1 Color=<rgba 0.6879 0.3357 0.816 1>, Variant 2 Color=?)
      Scale Factor=&MakeFloat4FromVectorAndFloat("Vector (RGB)"=$System.ToD, "Float (Alpha)"=$Particles.AlphaScale)
     )
     Scale Factor=&MakeFloat4FromVectorAndFloat("Vector (RGB)"=<1 1 1>, "Float (Alpha)"=1.5)
    )) {isSetParameters=true, module=SetVariables_#6}
   &InterpolateOverTime(Rate Of Change=3, Target Position Value:NiagaraPosition=$Particles.Position, Variable Type=::NewEnumerator6)
   /SetVariables(Particles.Interpolate Dynamic=&MakeFloat4FromVectorAndFloat(
     "Vector (RGB)"=$Particles.InterpolateOverTime.InterpolateOverTime_Position.Moving Average
     "Float (Alpha)"=&Add_Float(A=$User.Interpolate Intensity, B=-0.5)
    )) {isSetParameters=true, module=SetVariables_#7}
   &FN_Sprite_Attributes(
    Camera Offset=5
    Color=&MultiplyLinearColor(
     Linear Color=&VariantColorSelector(Variant Amount=0, Original Color=<rgba 0.0473 0.4323 0.3617 1>, Variant 1 Color=<rgba 0.2812 0.041 0.2452 1>, Variant 2 Color=?)
     Scale Factor=&MakeFloat4FromVectorAndFloat("Vector (RGB)"=$System.ToD, "Float (Alpha)"=&Multiply_Float(A=0.2, B=$Particles.AlphaScale))
    )
    Create Camera Offset Attribute=true
    Create Color Attribute=true
    Write to Module Namespace=true
   )
   /SetVariables(Particles.AirglowDyn:Vector4f=<2 1 0 0>) {isSetParameters=true, module=SetVariables_#8}
   /SetVariables(
    Particles.DecalDyn:Vector4f=&MakeVector4(
     X=0.8
     Y=0.5
     Z=&RandomRangeFloat(Minimum=0, Maximum=1)
     W=&RandomRangeFloat(Minimum=0, Maximum=0)
    )
    Particles.FireDecalColor:LinearColor=&MultiplyLinearColor(
     Linear Color=&Multiply_LinearColorByFloat(
      Linear Color=&VariantColorSelector(Variant Amount=0, Original Color=<rgba 0.1721 0.6323 0.7031 1>, Variant 1 Color=<rgba 0.6879 0.3357 0.816 1>, Variant 2 Color=?)
      Float=&Multiply_Float(
       A=1.75
       B=&SelectFloatFromArray(
        Direct Array Index=$Particles.PartitionParticles.Partition
        Float Selection Array=DataInterface() {propertyValues {FloatData [0.05,1,0.3,1,1,1.5,1.5]}}
       )
      )
     )
     Scale Factor=&MakeFloat4FromVectorAndFloat("Vector (RGB)"=$System.ToD, "Float (Alpha)"=$Particles.AlphaScale)
    )
    Particles.FireDecalSize:Vector2f=&Multiply_Vector2D_ByFloat(
     Float=&SelectFloatFromArray(
      Direct Array Index=$Particles.PartitionParticles.Partition
      Float Selection Array=DataInterface() {propertyValues {FloatData [0.3,0.5,0.5,0,0,0.5,1]}}
     )
     Vector2D=$Particles.SpriteSize
    )
    Particles.GlowDecalColor:LinearColor=&MultiplyLinearColor(
     Linear Color=&Multiply_LinearColorByFloat(
      Linear Color=&VariantColorSelector(Variant Amount=0, Original Color=<rgba 0.0054 0.1413 0.195 1>, Variant 1 Color=<rgba 0.2812 0.041 0.2452 1>, Variant 2 Color=?)
      Float=&Multiply_Float(A=1.75, B=$Particles.AlphaScale)
     )
     Scale Factor=&MakeFloat4FromVectorAndFloat("Vector (RGB)"=$System.ToD, "Float (Alpha)"=5)
    )
   ) {isSetParameters=true, module=SetVariables_#9}
  ]
  rendererClasses [&NiagaraMeshRendererProperties, &NiagaraSpriteRendererProperties]
  renderers [
   {rendererClass=&NiagaraMeshRendererProperties, rendererIndex=0}
   {rendererClass=&NiagaraMeshRendererProperties, rendererIndex=1}
   {rendererClass=&NiagaraMeshRendererProperties, rendererIndex=2}
   {rendererClass=&NiagaraMeshRendererProperties, rendererIndex=3}
   {rendererClass=&NiagaraMeshRendererProperties, rendererIndex=4}
   {rendererClass=&NiagaraMeshRendererProperties, rendererIndex=5}
   {rendererClass=&NiagaraSpriteRendererProperties, rendererIndex=6}
   {rendererClass=&NiagaraSpriteRendererProperties, rendererIndex=7}
   {rendererClass=&NiagaraSpriteRendererProperties, rendererIndex=8}
   {rendererClass=&NiagaraSpriteRendererProperties, rendererIndex=9}
   {rendererClass=&NiagaraSpriteRendererProperties, rendererIndex=10}
   {rendererClass=&NiagaraSpriteRendererProperties, rendererIndex=11}
  ]
 }
 BlowingParticles001 {
  emitterUpdateScript [
   &EmitterState(Life Cycle Mode=::NewEnumerator1, Loop Behavior=::NewEnumerator1, Loop Duration Mode=::NewEnumerator1, UseLoopDelay=false)
   &SpawnBurst_Instantaneous(
    Spawn Count=&MakeCustomIntFromBool(Bool=&Boolean_OrOperation(Bool1=$User.bIsFrontEnd, Bool2=$User.bIsFrontEndPreview), True INT=0, False INT=1)
    Use Spawn Probability=false
    UseLoopCountLimit=false
   )
   &SpawnRate(SpawnRate=15, Use Spawn Probability=false) {enabled=false}
  ]
  particleSpawnScript [&InitializeParticle(
    Color=&MultiplyLinearColor(
     Linear Color=&VariantColorSelector(Variant Amount=0, Original Color=<rgba 0.0018 0.0885 0.0857 1>, Variant 1 Color=<rgba 0.125 0.0182 0.1125 1>, Variant 2 Color=?)
     Scale Factor=&MakeFloat4FromVectorAndFloat("Vector (RGB)"=&Multiply_VectorByFloat(Vector=$System.ToD, Float=20), "Float (Alpha)"=0.15)
    )
    Lifetime Max=15
    Lifetime Min=12
    Lifetime Mode=::NewEnumerator1
    Mesh Scale Mode=::NewEnumerator3
    Sprite Rotation Mode=::NewEnumerator2
    Sprite Size Mode=::NewEnumerator1
    Uniform Sprite Size Max=8
    Uniform Sprite Size Min=4
    UsePositionOffset=false
    Write Material Random=true
   )]
  particleUpdateScript [
   &ParticleState(Kill Particles When Lifetime Has Elapsed=false, Let Infinitely Lived Particles Die When Emitter Deactivates=true)
   &SkeletalMeshLocation(Bone Sampling Mode=::NewEnumerator4, MaskSpawnGroup=false, Skeletal Mesh=$System.Skeletal Mesh)
   /SetVariables(Particles.AlphaScale:NiagaraFloat=&MakeCustomFloatFromBool(
     Bool=&Boolean_OrOperation(Bool1=$User.bIsFrontEndPreview, Bool2=$User.bIsFrontEnd)
     True Float=1
     False Float=&Lerp_Float(A=0, B=0.8, Alpha=$System.Timeline.NormalizedTime)
    )) {isSetParameters=true, module=SetVariables_#10}
   &ScaleColor(Scale Alpha=$Particles.AlphaScale, ScaleA=true, ScaleRGB=true)
  ]
  rendererClasses [&NiagaraSpriteRendererProperties, &NiagaraMeshRendererProperties]
  renderers [
   {rendererClass=&NiagaraSpriteRendererProperties}
   {rendererClass=&NiagaraMeshRendererProperties, rendererIndex=1}
   {rendererClass=&NiagaraMeshRendererProperties, rendererIndex=2}
   {rendererClass=&NiagaraMeshRendererProperties, rendererIndex=3}
   {rendererClass=&NiagaraMeshRendererProperties, rendererIndex=4}
   {rendererClass=&NiagaraMeshRendererProperties, rendererIndex=5}
  ]
 }
 BlowingParticlesFE {
  emitterUpdateScript [
   &EmitterState(
    Inactive Response=::NewEnumerator1
    Life Cycle Mode=::NewEnumerator1
    Loop Behavior=::NewEnumerator1
    Loop Duration Mode=::NewEnumerator1
    UseLoopDelay=false
   )
   &SpawnBurst_Instantaneous(
    Spawn Count=&MakeCustomIntFromBool(Bool=&Boolean_OrOperation(Bool1=$User.bIsFrontEnd, Bool2=$User.bIsFrontEndPreview), True INT=1, False INT=0)
    Use Spawn Probability=false
    UseLoopCountLimit=false
   )
   &SpawnRate(SpawnRate=15, Use Spawn Probability=false) {enabled=false}
  ]
  particleSpawnScript [&InitializeParticle(
    Color=&MultiplyLinearColor(
     Linear Color=&VariantColorSelector(Variant Amount=0, Original Color=<rgba 0.0018 0.0885 0.0857 1>, Variant 1 Color=<rgba 0.125 0.0182 0.1125 1>, Variant 2 Color=?)
     Scale Factor=&MakeFloat4FromVectorAndFloat("Vector (RGB)"=&Multiply_VectorByFloat(Vector=$System.ToD, Float=20), "Float (Alpha)"=0.15)
    )
    Lifetime=20
    Mesh Scale Mode=::NewEnumerator3
    Sprite Rotation Mode=::NewEnumerator2
    Sprite Size Mode=::NewEnumerator1
    Uniform Sprite Size Max=8
    Uniform Sprite Size Min=4
    UsePositionOffset=false
    Write Material Random=true
   )]
  particleUpdateScript [
   &ParticleState(
    Kill Particles When Lifetime Has Elapsed=false
    Let Infinitely Lived Particles Die When Emitter Deactivates=true
    Loop Particles Lifetime=true
   )
   &SkeletalMeshLocation(Bone Sampling Mode=::NewEnumerator4, MaskSpawnGroup=false, Skeletal Mesh=$System.Skeletal Mesh)
   /SetVariables(Particles.AlphaScale:NiagaraFloat=&FloatFromCurve(FloatCurve=DataInterface() {propertyValues {Curve {keys [
         {interpMode=RCIM_Cubic, tangentMode=RCTM_User, time=0.2, value=0.5}
         {arriveTangentWeight=0.1795, interpMode=RCIM_Cubic, leaveTangentWeight=0.1795, tangentMode=RCTM_User, time=0.4, value=1}
         {arriveTangentWeight=0.1795, interpMode=RCIM_Cubic, leaveTangentWeight=0.1795, tangentMode=RCTM_User, time=0.6, value=1}
         {tangentMode=RCTM_User, time=0.8, value=0.5}
        ]}}})) {isSetParameters=true, module=SetVariables_#10}
   &ScaleColor(Scale Alpha=$Particles.AlphaScale, ScaleA=true, ScaleRGB=true)
  ]
  rendererClasses [&NiagaraSpriteRendererProperties, &NiagaraMeshRendererProperties]
  renderers [
   {rendererClass=&NiagaraSpriteRendererProperties}
   {rendererClass=&NiagaraMeshRendererProperties, rendererIndex=1}
   {rendererClass=&NiagaraMeshRendererProperties, rendererIndex=2}
   {rendererClass=&NiagaraMeshRendererProperties, rendererIndex=3}
   {rendererClass=&NiagaraMeshRendererProperties, rendererIndex=4}
   {rendererClass=&NiagaraMeshRendererProperties, rendererIndex=5}
  ]
 }
 FlamesFE {
  emitterUpdateScript [
   &EmitterState(
    Inactive Response=::NewEnumerator1
    Life Cycle Mode=::NewEnumerator1
    Loop Behavior=::NewEnumerator1
    Loop Duration Mode=::NewEnumerator1
    UseLoopDelay=false
   )
   &SpawnBurst_Instantaneous(
    Spawn Count=&MakeCustomIntFromBool(Bool=&Boolean_OrOperation(Bool1=$User.bIsFrontEnd, Bool2=$User.bIsFrontEndPreview), True INT=7, False INT=0)
    Use Spawn Probability=false
    UseLoopCountLimit=false
   )
  ]
  particleSpawnScript [
   &PartitionParticles(Number Of Partitions=7)
   &InitializeParticle(
    Color=&MultiplyLinearColor(
     Linear Color=&VariantColorSelector(Variant Amount=0, Original Color=<rgba 0.0018 0.0885 0.0857 1>, Variant 1 Color=<rgba 0.125 0.0182 0.1125 1>, Variant 2 Color=?)
     Scale Factor=&MakeFloat4FromVectorAndFloat("Vector (RGB)"=$System.ToD, "Float (Alpha)"=1)
    )
    Lifetime=20
    Mesh Scale Mode=::NewEnumerator3
    Mesh Uniform Scale=&SelectFloatFromArray(
     Direct Array Index=$Particles.PartitionParticles.Partition
     Float Selection Array=DataInterface() {propertyValues {FloatData [0.6,0.35,0.2,0.7,0.5,0.9,0.4]}}
    )
    Sprite Size Mode=::NewEnumerator3
    Uniform Sprite Size=7
    UsePositionOffset=false
    Write Material Random=true
   )
   &InitialMeshOrientation(
    Mesh Axis To Orient=<0 0 1>
    Mesh Orientation Mode=::NewEnumerator2
    Orientation Coordinate Space=::World
    Orientation Vector=<0 0 1>
    Rotation=&SelectVectorFromArray(
     Direct Array Index=$Particles.PartitionParticles.Partition
     Vector Selection Array=DataInterface() {propertyValues {FloatData [<0 0 0>, <0 0 0>, <0 0 0>, <-0.09 0 0>, <-0.09 0 0>, <0 0 0>, <0 0 0>]}}
    )
    Use Rotation Vector=true
   )
   /SetVariables(Particles.VisibilityTag=&SelectIntFromArray(Direct Array Index=$Particles.PartitionParticles.Partition, Int Selection Array=DataInterface() {propertyValues {IntData [0,0,0,1,1,0,0]}})) {isSetParameters=true, module=SetVariables_#3}
   &DynamicMaterialParameters(
    Index 0 Param 1=&RandomRangeFloat(Minimum=0, Maximum=1)
    Index 0 Param 2=&SelectFloatFromArray(
     Direct Array Index=$Particles.PartitionParticles.Partition
     Float Selection Array=DataInterface() {propertyValues {FloatData [1.5,1,0.2,1.5,0.8,1.8,1.25]}}
    )
    Module.Index0Param0Write=true
    Module.Index0Param1Write=true
    Module.Index0Param2Write=true
    Module.Index0Param3Write=true
   )
  ]
  particleUpdateScript [
   &ParticleState(
    Kill Particles When Lifetime Has Elapsed=false
    Let Infinitely Lived Particles Die When Emitter Deactivates=true
    Loop Particles Lifetime=true
   )
   /SetVariables(Particles.AlphaScale:NiagaraFloat=&FloatFromCurve(FloatCurve=DataInterface() {propertyValues {Curve {keys [
         {interpMode=RCIM_Cubic, tangentMode=RCTM_User, time=0.2, value=0}
         {arriveTangentWeight=0.1795, interpMode=RCIM_Cubic, leaveTangentWeight=0.1795, tangentMode=RCTM_User, time=0.4, value=1}
         {arriveTangentWeight=0.1795, interpMode=RCIM_Cubic, leaveTangentWeight=0.1795, tangentMode=RCTM_User, time=0.6, value=1}
         {tangentMode=RCTM_User, time=0.8, value=0}
        ]}}})) {isSetParameters=true, module=SetVariables_#4}
   &ScaleMeshSize(Scale Factor=<1 1 2>)
   &SolveForcesAndVelocity(Clamp Velocity=false, Limit Acceleration=false)
   &SkeletalMeshLocation(
    Bone / Socket Index=&ReturnExecIndex()
    MaskSpawnGroup=false
    Mesh Sampling Type=::NewEnumerator5
    Orientation Sampling=::NewEnumerator1
    Sampled Position Offset=<0 0 -0.5>
    Skeletal Mesh=$System.Skeletal Mesh
    Socket Sampling Mode=::NewEnumerator5
   )
   &ScaleColor(Scale Alpha=$Particles.AlphaScale, Scale RGB=&VectorFromFloat(Value=3.5), ScaleA=true, ScaleRGB=true)
   /SetVariables(Particles.Color_02:LinearColor=&MultiplyLinearColor(
     Linear Color=&MultiplyLinearColor(
      Linear Color=&VariantColorSelector(Variant Amount=0, Original Color=<rgba 0.0018 0.0885 0.0857 1>, Variant 1 Color=<rgba 0.2812 0.041 0.2452 1>, Variant 2 Color=?)
      Scale Factor=&MakeFloat4FromVectorAndFloat("Vector (RGB)"=$System.ToD, "Float (Alpha)"=1)
     )
     Scale Factor=&MakeFloat4FromVectorAndFloat("Vector (RGB)"=<1 1 1>, "Float (Alpha)"=$Particles.AlphaScale)
    )) {isSetParameters=true, module=SetVariables_#5}
   /SetVariables(Particles.Color_03:LinearColor=&MultiplyLinearColor(
     Linear Color=&MultiplyLinearColor(
      Linear Color=&VariantColorSelector(Variant Amount=0, Original Color=<rgba 0.2203 0.9 0.7754 1>, Variant 1 Color=<rgba 0.6879 0.3357 0.816 1>, Variant 2 Color=?)
      Scale Factor=&MakeFloat4FromVectorAndFloat("Vector (RGB)"=$System.ToD, "Float (Alpha)"=$Particles.AlphaScale)
     )
     Scale Factor=&MakeFloat4FromVectorAndFloat("Vector (RGB)"=<1 1 1>, "Float (Alpha)"=1.5)
    )) {isSetParameters=true, module=SetVariables_#6}
   &InterpolateOverTime(Rate Of Change=3, Target Position Value:NiagaraPosition=$Particles.Position, Variable Type=::NewEnumerator6)
   /SetVariables(Particles.Interpolate Dynamic=&MakeFloat4FromVectorAndFloat(
     "Vector (RGB)"=$Particles.InterpolateOverTime.InterpolateOverTime_Position.Moving Average
     "Float (Alpha)"=&Add_Float(A=$User.Interpolate Intensity, B=-0.5)
    )) {isSetParameters=true, module=SetVariables_#7}
   &FN_Sprite_Attributes(
    Camera Offset=5
    Color=&MultiplyLinearColor(
     Linear Color=&VariantColorSelector(Variant Amount=0, Original Color=<rgba 0.0473 0.4323 0.3617 1>, Variant 1 Color=<rgba 0.2812 0.041 0.2452 1>, Variant 2 Color=?)
     Scale Factor=&MakeFloat4FromVectorAndFloat("Vector (RGB)"=$System.ToD, "Float (Alpha)"=&Multiply_Float(A=0.2, B=$Particles.AlphaScale))
    )
    Create Camera Offset Attribute=true
    Create Color Attribute=true
    Write to Module Namespace=true
   )
   /SetVariables(Particles.AirglowDyn:Vector4f=<2 1 0 0>) {isSetParameters=true, module=SetVariables_#8}
   /SetVariables(
    Particles.DecalDyn:Vector4f=&MakeVector4(
     X=0.8
     Y=0.5
     Z=&RandomRangeFloat(Minimum=0, Maximum=1)
     W=&RandomRangeFloat(Minimum=0, Maximum=0)
    )
    Particles.FireDecalColor:LinearColor=&MultiplyLinearColor(
     Linear Color=&Multiply_LinearColorByFloat(
      Linear Color=&VariantColorSelector(Variant Amount=0, Original Color=<rgba 0.1721 0.6323 0.7031 1>, Variant 1 Color=<rgba 0.6879 0.3357 0.816 1>, Variant 2 Color=?)
      Float=&Multiply_Float(
       A=1.75
       B=&SelectFloatFromArray(
        Direct Array Index=$Particles.PartitionParticles.Partition
        Float Selection Array=DataInterface() {propertyValues {FloatData [0.05,1,0.3,1,1,1.5,1.5]}}
       )
      )
     )
     Scale Factor=&MakeFloat4FromVectorAndFloat("Vector (RGB)"=$System.ToD, "Float (Alpha)"=$Particles.AlphaScale)
    )
    Particles.FireDecalSize:Vector2f=&Multiply_Vector2D_ByFloat(
     Float=&SelectFloatFromArray(
      Direct Array Index=$Particles.PartitionParticles.Partition
      Float Selection Array=DataInterface() {propertyValues {FloatData [0.3,0.5,0.5,0,0,0.5,1]}}
     )
     Vector2D=$Particles.SpriteSize
    )
    Particles.GlowDecalColor:LinearColor=&MultiplyLinearColor(
     Linear Color=&Multiply_LinearColorByFloat(
      Linear Color=&VariantColorSelector(Variant Amount=0, Original Color=<rgba 0.0054 0.1413 0.195 1>, Variant 1 Color=<rgba 0.2812 0.041 0.2452 1>, Variant 2 Color=?)
      Float=&Multiply_Float(A=1.75, B=$Particles.AlphaScale)
     )
     Scale Factor=&MakeFloat4FromVectorAndFloat("Vector (RGB)"=$System.ToD, "Float (Alpha)"=5)
    )
   ) {isSetParameters=true, module=SetVariables_#9}
   &FN_Sprite_Attributes(Camera Offset=1, Create Camera Offset Attribute=true, Write to Module Namespace=true) {module=FN_Sprite_Attributes001}
   /SetVariables(Particles.RotatedMeshOrientation:Quat4f=&MultiplyQuaternion(Quaternion A=$Particles.MeshOrientation, Quaternion B=$Particles.SkeletalMeshLocation.SampledRotation)) {isSetParameters=true}
  ]
  rendererClasses [&NiagaraMeshRendererProperties, &NiagaraSpriteRendererProperties]
  renderers [
   {rendererClass=&NiagaraMeshRendererProperties, rendererIndex=0}
   {rendererClass=&NiagaraMeshRendererProperties, rendererIndex=1}
   {rendererClass=&NiagaraMeshRendererProperties, rendererIndex=2}
   {rendererClass=&NiagaraMeshRendererProperties, rendererIndex=3}
   {rendererClass=&NiagaraMeshRendererProperties, rendererIndex=4}
   {rendererClass=&NiagaraMeshRendererProperties, rendererIndex=5}
   {rendererClass=&NiagaraSpriteRendererProperties, rendererIndex=6}
   {rendererClass=&NiagaraSpriteRendererProperties, rendererIndex=7}
   {rendererClass=&NiagaraSpriteRendererProperties, rendererIndex=8}
   {rendererClass=&NiagaraSpriteRendererProperties, rendererIndex=9}
   {rendererClass=&NiagaraSpriteRendererProperties, rendererIndex=10}
   {rendererClass=&NiagaraSpriteRendererProperties, rendererIndex=11}
  ]
 }
}
summary {
 systemName=NS_Pickaxe_PolishedJade_Idle
 emitters [
  {
   emitterName=BlowingParticles
   rendererClasses [&NiagaraMeshRendererProperties, &NiagaraSpriteRendererProperties]
  }
  {
   emitterName=BlowingParticles001
   rendererClasses [&NiagaraSpriteRendererProperties, &NiagaraMeshRendererProperties]
  }
  {emitterName=FlamesFE, rendererClasses [&NiagaraMeshRendererProperties, &NiagaraSpriteRendererProperties]}
  {
   emitterName=BlowingParticlesFE
   rendererClasses [&NiagaraSpriteRendererProperties, &NiagaraMeshRendererProperties]
  }
 ]
 userVariables [
  User.bIsFrontEnd {type=NiagaraBool}
  User.bIsFrontEndPreview {type=NiagaraBool}
  User.Interpolate Intensity {type=NiagaraFloat, defaultValue=1, description=""}
  User.RibbonIDOffset {type=NiagaraInt32}
  User.Variant {type=NiagaraFloat, defaultValue=0, description=""}
 ]
}
