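# Material for the Polished Jade pickaxe that layers animated emissive jade effects between the shared 2023 character PreFX and PostFX functions.
# Rotated, panning noise builds a smoke mask, while noise-distorted plasma coils form energy lines.
# These animated layers combine with mask and emissive textures, tintable smoke color and an edge glow.
# A UV-space sphere mask and a mask channel shape final brightness, replacing only Emissive Color in the material attributes.
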Material /BRCosmetics/Weapons/FORT_Melee/Pickaxe_PolishedJade/Materials/M_PIckaxe_PolishedJade
graph {
 properties {
  bUseMaterialAttributes=true
  bUsedWithMeshDeformer=true
  bUsedWithMorphTargets=true
  bUsedWithSkeletalMesh=true
  pixelDepthOffsetMode=PDOM_Legacy
  shadingModel=MSM_FromMaterialExpression
  usedShadingModels=Default Lit | Subsurface | Clear Coat | Subsurface Profile | Cloth
 }
 nodes [
  n1 = MaterialFunctionCall() {materialFunction=&MF_CharacterParent_2023_PreFX}
  MaterialFunctionCall(
   Material=SetMaterialAttributes(
    MaterialAttributes=GetMaterialAttributes(None=@n1.Result) {attributeGetTypes [#1]}.MaterialAttributes
    Emissive Color=NamedRerouteUsage()
   ) {attributeSetTypes [#1]}
   M=@n1.M
   Pre Skinned Local Position=@n1.M
   Distance Blend=@n1.M
   Fresnel=@n1.M
  ) {materialFunction=&MF_CharacterParent_2023_PostFX}
  n2 = NamedRerouteUsage()
  n3 = TextureSample(UVs=Add(A=Multiply(A=@n2) {constB=0.06}, B=TextureCoordinate())) {
   samplerSource=SSM_Wrap_WorldGroupSettings
   texture=/BRCosmetics/Weapons/FORT_Melee/Pickaxe_PolishedJade/Textures/T_Pickaxe_PolishedJade_E_Mask
  }
  n4 = Multiply(A=LinearInterpolate(Alpha=NamedRerouteUsage()) {constA=0.5}, B=Multiply(A=@n3.G) {constB=1.2})
  n5 = TextureSampleParameter2D() {
   group=PolishedJade_Pickaxe
   parameterName=EmissiveColor
   samplerSource=SSM_Wrap_WorldGroupSettings
   texture=/BRCosmetics/Weapons/FORT_Melee/Pickaxe_PolishedJade/Textures/T_Pickaxe_PolishedJade_E
  }
  n6 = StaticComponentMaskParameter(None=@n5.RGB) {defaultB=true, group=PolishedJade_Pickaxe, parameterName=Mask channel}
  n7 = VectorParameter() {group=PolishedJade_Pickaxe, parameterName=SmokeColor, defaultValue=<rgba 0.2 3.4996 4 1>}
  n8 = Multiply(A=Multiply(A=Multiply(A=@n4, B=@n6)) {constB=1.5}, B=@n7.RGB)
  n9 = Multiply(
   A=Multiply(A=@n7.RGB, B=ScalarParameter() {defaultValue=0.5, group=PolishedJade_Pickaxe, parameterName=Energy indensity}) {constB=0.5}
   B=Multiply(A=NamedRerouteUsage(), B=Saturate(None=@n6))
  )
  n10 = TextureCoordinate()
  n11 = Constant()
  n12 = Constant() {r=1.1}
  n13 = Constant() {r=0.25}
  n14 = Multiply(
   A=Add(A=Add(A=@n8, B=@n9), B=NamedRerouteUsage())
   B=LinearInterpolate(A=Constant() {r=2}, B=Constant() {r=1.2}, Alpha=SphereMask(A=@n10, B=@n11, Radius=@n12, Hardness=@n13))
  ) {constB=1.32}
  n15 = TextureSample() {
   samplerSource=SSM_Wrap_WorldGroupSettings
   texture=&T_Pickaxe_PolishedJade_E_Mask
  }
  n16 = LinearInterpolate(A=@n14, B=Multiply(A=@n14, B=Constant() {r=1.5}) {constB=3}, Alpha=@n15.B)
  MaterialFunctionCall(Emissive=@n16) {materialFunction=&MF_FX_EmissiveAdjustments}
  n17 = Multiply(A=TextureCoordinate() {uTiling=0.8}) {constB=3}
  n18 = Constant() {r=0.15}
  n19 = TextureSample(UVs=Panner(Coordinate=MaterialFunctionCall(UVs=@n17, "Rotation Angle (0-1)"=@n18) {materialFunction=&CustomRotator}.Rotated Values) {speedY=0.3}) {samplerSource=SSM_Wrap_WorldGroupSettings, samplerType=SAMPLERTYPE_Grayscale, texture=&LowResBlurredNoise}
  n20 = Multiply(A=TextureCoordinate() {uTiling=0.8}) {constB=8}
  n21 = TextureSample(UVs=Panner(Coordinate=MaterialFunctionCall(UVs=@n20, "Rotation Angle (0-1)"=@n18) {materialFunction=&CustomRotator}.Rotated Values) {speedY=0.6}) {samplerSource=SSM_Wrap_WorldGroupSettings, samplerType=SAMPLERTYPE_Grayscale, texture=&LowResBlurredNoise}
  NamedRerouteDeclaration(None=Max(A=@n19.R, B=@n21.R)) {name=Smoke, nodeColor=<rgba 0.0547 0 1 1>}
  NamedRerouteDeclaration(None=@n16) {name=out_Emissive, nodeColor=<rgba 0.9922 1 0 1>}
  n22 = MaterialFunctionCall(UVs=Multiply(A=TextureCoordinate()) {constB=2}, "Rotation Angle (0-1)"=Constant() {r=0.15}) {materialFunction=&CustomRotator}
  n23 = TextureSample(UVs=Panner(Coordinate=@n22.Rotated Values) {speedY=0.1}) {samplerSource=SSM_Wrap_WorldGroupSettings, samplerType=SAMPLERTYPE_Grayscale, texture=&LowResBlurredNoise}
  n24 = LinearInterpolate(Alpha=@n23.R) {constA=0.6}
  n25 = TextureSample(UVs=Panner(Coordinate=@n22.Rotated Values) {speedY=0.3}) {samplerSource=SSM_Wrap_WorldGroupSettings, samplerType=SAMPLERTYPE_Grayscale, texture=&LowResBlurredNoise}
  n26 = TextureSample(UVs=Add(A=Panner(Coordinate=@n22.Rotated Values) {speedY=0.2}, B=Multiply(A=@n25.R) {constB=0.35})) {
   samplerSource=SSM_Wrap_WorldGroupSettings
   samplerType=SAMPLERTYPE_LinearGrayscale
   texture=/Game/Effects/Fort_Effects/Emotes/UltraLight/Texture/T_PlasmaCoils
  }
  NamedRerouteDeclaration(None=Power(Base=Multiply(A=@n24, B=@n26.R), Exp=Constant() {r=1.2}) {constExponent=1.2}) {name=Energy line, nodeColor=<rgba 0.2656 1 0 1>}
  n27 = TextureSampleParameter2D() {
   group=PolishedJade_Pickaxe
   parameterName=EmissiveColor
   samplerSource=SSM_Wrap_WorldGroupSettings
   texture=&T_Pickaxe_PolishedJade_E
  }
  n28 = ScalarParameter() {defaultValue=0.2, group=PolishedJade_Pickaxe, parameterName=Base color intensity}
  n29 = TextureSample() {
   samplerSource=SSM_Wrap_WorldGroupSettings
   texture=&T_Pickaxe_PolishedJade_E_Mask
  }
  n30 = LinearInterpolate(B=Constant() {r=0.5}, Alpha=@n29.R) {constB=2}
  n31 = VectorParameter() {group=PolishedJade_Pickaxe, parameterName=EdgeColor, defaultValue=<rgba 0.0378 0.4531 0.4531 1>}
  NamedRerouteDeclaration(None=Add(A=Multiply(A=@n27.RGB, B=@n28) {constB=0.5}, B=Multiply(A=@n30, B=@n31.RGB))) {name=EdgeGlow, nodeColor=<rgba 1 0 0.5859 1>}
 ]
}
