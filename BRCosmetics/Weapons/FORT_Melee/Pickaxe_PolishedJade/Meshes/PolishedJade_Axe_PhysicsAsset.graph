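# Physics asset for the Polished Jade pickaxe, a two-handed melee weapon cosmetic, giving its axe mesh a simple collision shape.
# It holds a single capsule body on the Melee_TwoHanded_Handle bone, offset along and rotated to fit the weapon.
# There are no constraints, so the weapon is represented as one body with no jointed parts.
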PhysicsAsset /BRCosmetics/Weapons/FORT_Melee/Pickaxe_PolishedJade/Meshes/PolishedJade_Axe_PhysicsAsset
bodies {Melee_TwoHanded_Handle [{shapeType=Capsule, length=33.4227, radius=13.0149, center=<0.0001 -0.1423 17.1359>, rotation=<rot 1.2546 -90 0.0009>}]}
constraints []
